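# Looping mechanical foley sound cue for the MiddleSock emote.
# A random node picks one of three equally weighted looping waves.
# A continuous modulator drives volume and slight pitch variation from the EmoteMoveSpeed parameter, so the sound follows emote movement.
# It uses emote attenuation, the emote foley sound class, a three-voice concurrency override and a send to the music audio analysis submix.
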SoundCue /BRCosmetics/Sounds/Emotes/MiddleSock/EmoteFoley_MiddleSock_Mech_Cue
root=ModulatorContinuous_0
nodes [
 {
  id=ModulatorContinuous_0
  children [Random_0]
  properties {
   pitchModulationParams {maxOutput=1.02, minOutput=0.98, parameterName=EmoteMoveSpeed}
   volumeModulationParams {maxOutput=0.4, parameterName=EmoteMoveSpeed}
  }
 }
 {id=Random_0, children [WavePlayer_0,WavePlayer_1,WavePlayer_2], properties {weights [1,1,1]}}
 {id=WavePlayer_0, properties {bLooping=true, soundWaveAssetPtr=&/IP_Emote_MiddleSock_Mech_LP_01}}
 {id=WavePlayer_1, properties {bLooping=true, soundWaveAssetPtr=&/IP_Emote_MiddleSock_Mech_LP_02}}
 {id=WavePlayer_2, properties {bLooping=true, soundWaveAssetPtr=&/IP_Emote_MiddleSock_Mech_LP_03}}
]
properties {
 attenuationSettings=&Fort_Emotes_Athena_Att
 audioPropertiesSheet=/BRCosmetics/Sounds/Settings/PropertySheets/Emotes/APS_Cosmetic_Emote_Foley
 bOverrideConcurrency=true
 maxDistance=1600
 soundClassObject=/Game/Sounds/FortSoundClassesAndModes/Fort_Foley_Emote
 volumeMultiplier=1
 concurrencyOverrides {maxCount=3}
 soundSubmixSends [{minSendLevel=0, soundSubmix=&Fortnite_AudioAnalysis_Music}]
}
